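# Sound cue for the end of the SunRose Harbinger SMG reload, classed as weapon foley and using weapon reload attenuation.
# A mixer layers one of three equally weighted random reload-end takes with a slightly quieter sweetener wave.
# Each layer passes through its own modulator, and the random take also runs through a delay node.
# The cue overrides concurrency and has its debug flag enabled.
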SoundCue /SunRoseWeaponsGameplay/Items/Weapons/HarbingerSMG/Sounds/SunRose_SMG_Reload_End_Cue
root=Mixer_0
nodes [
 {id=Delay_0, children [Modulator_0]}
 {id=Mixer_0, children [Delay_0,Modulator_1], properties {inputVolume [1,0.8]}}
 {id=Modulator_0, children [Random_2]}
 {id=Modulator_1, children [WavePlayer_5]}
 {id=Random_2, children [WavePlayer_0,WavePlayer_3,WavePlayer_4], properties {weights [1,1,1]}}
 {id=WavePlayer_0, properties {soundWaveAssetPtr=&/SunRose_SMG_Reload_End_01}}
 {id=WavePlayer_3, properties {soundWaveAssetPtr=&/SunRose_SMG_Reload_End_02}}
 {id=WavePlayer_4, properties {soundWaveAssetPtr=&/SunRose_SMG_Reload_End_03}}
 {id=WavePlayer_5, properties {soundWaveAssetPtr=&/SunRose_SMG_ReloadEnd_Sweetner}}
]
properties {
 attenuationSettings=&Fort_Weapon_Reload_Att
 bDebug=true
 bOverrideConcurrency=true
 maxDistance=2250
 soundClassObject=&Fort_WeaponFoley
 volumeMultiplier=0.501
}
