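# Sound cue for the start of the Harbinger SMG reload, playing a side-slap foley sound.
# A Random node picks one of three equally weighted SideSlap wave variants and feeds a Modulator at the root for playback variation.
# It plays at reduced volume in the weapon foley sound class, using shared reload attenuation and a concurrency override, with debug enabled.
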SoundCue /SunRoseWeaponsGameplay/Items/Weapons/HarbingerSMG/Sounds/SunRose_SMG_Reload_Start_Cue
root=Modulator_1
nodes [
 {id=Modulator_1, children [Random_0]}
 {id=Random_0, children [WavePlayer_1,WavePlayer_2,WavePlayer_6], properties {weights [1,1,1]}}
 {id=WavePlayer_1, properties {soundWaveAssetPtr=&/SunRose_SMG_SideSlap_01}}
 {id=WavePlayer_2, properties {soundWaveAssetPtr=&/SunRose_SMG_SideSlap_02}}
 {id=WavePlayer_6, properties {soundWaveAssetPtr=&/SunRose_SMG_SideSlap_03}}
]
properties {
 attenuationSettings=&Fort_Weapon_Reload_Att
 bDebug=true
 bOverrideConcurrency=true
 maxDistance=2250
 soundClassObject=&Fort_WeaponFoley
 volumeMultiplier=0.3
}
